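# Animation Blueprint for the female Space Fighter backpack cosmetic, whose output pose is just a looping Idle sequence.
# A root-bone zero-scale node driven by RootScaleAlpha appears meant to hide the pack under a matching glider, but does not reach the output pose.
# EventGraph logic sets that alpha from skydiving, parachute and GliderMatch flags, but the update event is unwired.
# GliderMatch is never set, and the glider mesh reference is unused.
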AnimBlueprint /BRCosmetics/Accessories/FORT_Backpacks/Backpack_F_MED_Space_Fighter/Meshes/F_MED_Space_Fighter_Pack_AnimBP
target_skeleton=&/F_MED_Space_Fighter_Pack_Skeleton
graphs [
 AnimGraph {nodes [
   n1 = "Transform (Modify) Bone - Bone: root"(
    ComponentPose=Local To Component()
    Alpha=Get RootScaleAlpha() -> (RootScaleAlpha:double)
    Translation="0,0,0"
    Rotation="0,0,0"
    Scale="0,0,0"
   ) {scaleMode=BMM_Replace, alphaBoolBlend {blendOption=Linear}, boneToModify {boneName=root}, scale=<0 0 0>}
   Component To Local(ComponentPose=@n1)
   Input Pose()
   Output Pose(Result=Sequence Player 'Idle'() {sequence=/BRCosmetics/Accessories/FORT_Backpacks/Backpack_F_MED_Space_Fighter/Anims/Idle}) {layerGroup=None}
  ]}
 EventGraph {nodes [
   n1 = Sequence()
   > Do Once()
   n2 = Break Fort Anim Input Skydiving!(FortAnimInput_Skydiving:"Fort Anim Input Skydiving Structure (by ref)"=Get Skydiving!(self:Fort Player Anim Instance Object Reference) -> (Skydiving:Fort Anim Input Skydiving Structure)) -> (bIsParachuteOpen:bool=~, bIsSkydiving:bool=~, bHasDeployedAndIsStandingUpright:bool=~, PoseAdditive:Anim Sequence Object Reference) {is=title K2Node_BreakStruct}
   n3 = Branch(execute=@n1.then_1, Condition=@n2.bIsSkydiving)
   > n4 = Branch(Condition=AND Boolean(A=@n2, B=Get GliderMatch() -> (GliderMatch:bool)))
   > Set RootScaleAlpha(RootScaleAlpha:double=1) -> (Output_Get:double)
   Set RootScaleAlpha(execute=[@n4.else, @n3.else], RootScaleAlpha=0)
   TryGetPawnOwner()
   Event BlueprintUpdateAnimation()
  ]}
]
variables [
 GliderMatch {type=boolean}
 RootScaleAlpha {type=number}
 SpaceFighter_Glider_Ref {type=SkeletalMesh, default=/BRCosmetics/Gadgets/Assets/VinderTech_GliderChute/Glider_Space_Fighter_Female/Meshes/Space_Fighter_Female_Glider}
]
