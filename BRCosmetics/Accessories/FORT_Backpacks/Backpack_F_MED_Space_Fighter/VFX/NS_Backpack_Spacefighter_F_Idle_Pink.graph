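# Pink idle Niagara effect for the female Space Fighter backpack: electric-looking arcs, glows and flickering lights attached to skeletal mesh sockets.
# Two BoneFlow ribbon emitters burst partitioned particles at random times along upper and lower socket sets, bowing arcs with a radius curve and flashing them bright.
# Spawn-group sprites add sparks and soft glow at sockets, while two offset emitters flicker particle lights.
# User.ColorChange supplies the shared pink tint.
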NiagaraSystem /BRCosmetics/Accessories/FORT_Backpacks/Backpack_F_MED_Space_Fighter/VFX/NS_Backpack_Spacefighter_F_Idle_Pink
dependencies {
 usedDataInterfaces [&NiagaraDataInterfaceCurve, &NiagaraDataInterfaceColorCurve]
 usedDynamicInputs []
 usedModules [&SystemState]
}
emitters {
 BoneFlow_ArcBelow {
  emitterSpawnScript [
   /SetVariables(Emitter.RandomSeed:NiagaraInt32=/Niagara/DynamicInputs/Random/RandomInt(Randomness Mode:ENiagaraRandomnessMode=::NewEnumerator0, Random Seed:NiagaraInt32=0, Fixed Random Seed:NiagaraBool=false)) {isSetParameters=true, module=SetVariables_#1}
   /Game/Effects/Niagara/Modules/Partition/FN_PartitionAttributes(Number of Bone Sets=1, Particles Per Partition=8)
  ]
  emitterUpdateScript [
   &EmitterState(Life Cycle Mode=::NewEnumerator1, Loop Duration=0.6, UseLoopDelay=false)
   &SpawnBurst_Instantaneous(
    Spawn Count=$Emitter.FN_PartitionAttributes.Spawn Total
    Spawn Time=&RandomRangeFloat(Minimum=0, Maximum=0.3)
    Use Spawn Probability=false
   )
  ]
  particleSpawnScript [
   /Game/Effects/Niagara/Modules/Partition/FN_Particle_partition_v5/FN_Particle_Partition_v5(Enable Ribbon ID=true, Pattern:ENiagaraPartitionWaveForm=::NewEnumerator1, Seed:NiagaraInt32=$Emitter.RandomSeed)
   /Niagara/Modules/Spawn/Initialization/InitializeParticle(
    Lifetime=5
    Module.Write Color=true
    Module.Write Lifetime=false
    Module.Write Mass=true
    Module.Write Material Random=true
    Module.Write Position=false
    Module.Write Scale=false
    Module.Write SpriteRotation=false
    Module.Write SpriteSize=true
    Position=&SimulationPosition()
    Sprite Size=<1 1>
   ) {enabled=false}
   /Game/Effects/Niagara/Modules/BoneFlow/FN_Initialize_BoneFlow_Attributes_v2(
    Angle:NiagaraFloat=&RandomRangeFloat(Minimum=0.25, Maximum=0.35)
    Radius:NiagaraFloat=&FloatFromCurve(
     FloatCurve=DataInterface() {propertyValues {Curve {keys [
         {arriveTangent=10.1407, interpMode=RCIM_Cubic, leaveTangent=10.1407, tangentMode=RCTM_User, time=0, value=0}
         <at 0.4983 2 RCIM_Cubic>
         {arriveTangent=-9.1617, interpMode=RCIM_Cubic, leaveTangent=-9.1617, tangentMode=RCTM_User, time=1, value=0}
        ]}}}
     CurveIndex=$Particles.Partition.NormalizedID
    )
    T:NiagaraFloat=&UniformRangedFloat(Minimum=0, Maximum=1)
   )
   &InitializeRibbon(
    Color=$User.ColorChange
    Lifetime=&UniformRangedFloat(Minimum=0.2, Maximum=0.5)
    Module.Write Color=true
    Module.Write Lifetime=true
    Module.Write Mass=true
    Module.Write Material Random=true
    Module.Write Position=true
    Module.Write RibbonTwist=false
    Module.Write RibbonWidth=true
    Ribbon Width=&RandomRangeFloat(Minimum=6, Maximum=13)
   )
   &DynamicMaterialParameters(
    Index 0 Param 3=&RandomRangeFloat(Minimum=0.4, Maximum=0.6)
    Module.Param0WriteEnabled=false
    Module.Param1WriteEnabled=true
    Module.Param2WriteEnabled=true
    Module.Param3WriteEnabled=false
   ) {module=DynamicMaterialParameters001}
  ]
  particleUpdateScript [
   &ParticleState()
   &FN_BoneFlow(
    Angle:NiagaraFloat=$Particles.Initial.FN_Initialize_BoneFlow_Attributes_v2.Angle
    Height:NiagaraFloat=$Particles.Initial.Partition.NormalizedID
    Radius:NiagaraFloat=$Particles.Initial.FN_Initialize_BoneFlow_Attributes_v2.Radius
    SkeletalMesh:NiagaraDataInterfaceSkeletalMesh=$User.SkeletalMeshBelow
    Use Sockets Instead=true
    Write Position Attribute=true
   )
   &ScaleColor(
    Scale Mode=::NewEnumerator1
    Scale RGBA=&Vector4FromCurve(Vector4Curve=DataInterface() {propertyValues {
       AlphaCurve {keys [<at 0 0>,<at 0.2262 1>,<at 0.4885 1>,<at 1 1>]}
       BlueCurve {keys [<at 0.3099 30>,<at 0.3685 5>,<at 0.9991 1>]}
       GreenCurve {keys [<at 0.3099 30>,<at 0.3685 5>,<at 0.9991 1>]}
       RedCurve {keys [<at 0.3099 30>,<at 0.3685 5>,<at 0.9991 1>]}
      }})
    ScaleRGBA=true
   )
  ]
  rendererClasses [&NiagaraRibbonRendererProperties]
  renderers [{rendererClass=&NiagaraRibbonRendererProperties}]
 }
 BoneFlow_ArcUp {
  emitterSpawnScript [
   /SetVariables(Emitter.RandomSeed:NiagaraInt32=&RandomInt(Randomness Mode:ENiagaraRandomnessMode=::NewEnumerator0, Random Seed:NiagaraInt32=0, Fixed Random Seed:NiagaraBool=false)) {isSetParameters=true, module=SetVariables_#1}
   &FN_PartitionAttributes(Number of Bone Sets=1, Particles Per Partition=8)
  ]
  emitterUpdateScript [
   &EmitterState(Life Cycle Mode=::NewEnumerator1, Loop Duration=0.6, UseLoopDelay=false)
   &SpawnBurst_Instantaneous(
    Spawn Count=$Emitter.FN_PartitionAttributes.Spawn Total
    Spawn Time=&RandomRangeFloat(Minimum=0, Maximum=0.3)
    Use Spawn Probability=false
   )
  ]
  particleSpawnScript [
   &FN_Particle_Partition_v5(Enable Ribbon ID=true, Pattern:ENiagaraPartitionWaveForm=::NewEnumerator1, Seed:NiagaraInt32=$Emitter.RandomSeed)
   &InitializeParticle(
    Color=$User.ColorChange
    Lifetime=5
    Module.Write Color=true
    Module.Write Lifetime=true
    Module.Write Mass=true
    Module.Write Material Random=true
    Module.Write Position=false
    Module.Write Scale=false
    Module.Write SpriteRotation=false
    Module.Write SpriteSize=true
    Position=&SimulationPosition()
    Sprite Size=<1 1>
   ) {enabled=false}
   &FN_Initialize_BoneFlow_Attributes_v2(
    Angle:NiagaraFloat=&RandomRangeFloat(Minimum=0.25, Maximum=0.35)
    Radius:NiagaraFloat=&FloatFromCurve(
     FloatCurve=DataInterface() {propertyValues {Curve {keys [
         {arriveTangent=10.1407, interpMode=RCIM_Cubic, leaveTangent=10.1407, tangentMode=RCTM_User, time=0, value=0}
         <at 0.4983 2 RCIM_Cubic>
         {arriveTangent=-9.1617, interpMode=RCIM_Cubic, leaveTangent=-9.1617, tangentMode=RCTM_User, time=1, value=0}
        ]}}}
     CurveIndex=$Particles.Partition.NormalizedID
    )
    T:NiagaraFloat=&UniformRangedFloat(Minimum=0, Maximum=1)
   )
   &InitializeRibbon(
    Color=$User.ColorChange
    Lifetime=&UniformRangedFloat(Minimum=0.2, Maximum=0.5)
    Module.Write Color=true
    Module.Write Lifetime=true
    Module.Write Mass=true
    Module.Write Material Random=true
    Module.Write Position=true
    Module.Write RibbonTwist=false
    Module.Write RibbonWidth=true
    Ribbon Width=&RandomRangeFloat(Minimum=6, Maximum=13)
   )
   &DynamicMaterialParameters(
    Index 0 Param 1=1
    Index 0 Param 3=&RandomRangeFloat(Minimum=0.4, Maximum=0.6)
    Module.Param0WriteEnabled=false
    Module.Param1WriteEnabled=true
    Module.Param2WriteEnabled=true
    Module.Param3WriteEnabled=false
   ) {module=DynamicMaterialParameters001}
  ]
  particleUpdateScript [
   &ParticleState()
   &FN_BoneFlow(
    Angle:NiagaraFloat=$Particles.Initial.FN_Initialize_BoneFlow_Attributes_v2.Angle
    Height:NiagaraFloat=$Particles.Initial.Partition.NormalizedID
    Radius:NiagaraFloat=$Particles.Initial.FN_Initialize_BoneFlow_Attributes_v2.Radius
    SkeletalMesh:NiagaraDataInterfaceSkeletalMesh=$User.SkeletalMeshUp
    Use Sockets Instead=true
    Write Position Attribute=true
   )
   &ScaleColor(
    Scale Mode=::NewEnumerator1
    Scale RGBA=&Vector4FromCurve(Vector4Curve=DataInterface() {propertyValues {
       AlphaCurve {keys [<at 0 0>,<at 0.2262 1>,<at 0.4885 1>,<at 1 1>]}
       BlueCurve {keys [<at 0.3099 30>,<at 0.3685 5>,<at 0.9991 1>]}
       GreenCurve {keys [<at 0.3099 30>,<at 0.3685 5>,<at 0.9991 1>]}
       RedCurve {keys [<at 0.3099 30>,<at 0.3685 5>,<at 0.9991 1>]}
      }})
    ScaleRGBA=true
   )
  ]
  rendererClasses [&NiagaraRibbonRendererProperties, &NiagaraSpriteRendererProperties]
  renderers [
   {rendererClass=&NiagaraRibbonRendererProperties}
   {rendererClass=&NiagaraSpriteRendererProperties, rendererIndex=1}
  ]
 }
 Burst001 {
  emitterUpdateScript [
   &EmitterState()
   &SpawnRate(SpawnRate=6, Use Spawn Probability=false)
  ]
  particleSpawnScript [
   &InitializeParticle(
    Color=&MultiplyLinearColor(Linear Color=$User.ColorChange, Scale Factor=<2 2 2 1>)
    Lifetime=&RandomRangeFloat(Minimum=0.2, Maximum=0.8)
    Module.Write Color=true
    Module.Write Lifetime=true
    Module.Write Mass=true
    Module.Write Material Random=true
    Module.Write Position=false
    Module.Write Scale=false
    Module.Write SpriteRotation=true
    Module.Write SpriteSize=true
    Position=&SimulationPosition()
    Sprite Rotation=&UniformRangedFloat(Minimum=0, Maximum=360)
    Sprite Size=&Vector2DFromFloat(Value=&UniformRangedFloat(Minimum=5, Maximum=8))
   )
   &CameraOffset(Camera Offset Amount=&UniformRangedFloat(Minimum=0, Maximum=3.5))
   &FN_SetSpawnGroup_v2(Number of Spawn Groups=4)
  ]
  particleUpdateScript [
   &ParticleState()
   /Niagara/Modules/Update/SubUV/SubUVAnimation(SubUV Animation Mode=::NewEnumerator1)
   &SkeletalMeshLocation(
    Bone / Socket Index=$Particles.SpawnGroup
    Inherit Velocity=false
    MaskSpawnGroup=false
    Mesh Sampling Type=::NewEnumerator5
    Sampled Position Offset=&RandomRangeVector(Minimum=<1 1 1>, Maximum=<-1 -1 -1>)
    Skeletal Mesh=$User.SkltnMesh
    Socket Sampling Mode=::NewEnumerator5
   )
   &ScaleColor(
    Scale Mode=::NewEnumerator1
    Scale RGBA=&Vector4FromCurve(Vector4Curve=DataInterface() {propertyValues {
       AlphaCurve {keys [<at 0 0>,<at 0.0532 1>,<at 0.9516 1>,<at 1 0>]}
       BlueCurve {keys [<at 0 10>,<at 0.2513 10>,<at 0.3322 1>,<at 0.5108 1>,<at 0.5778 2>,<at 0.9182 2>,<at 1 1>]}
       GreenCurve {keys [<at 0 10>,<at 0.2513 10>,<at 0.3322 1>,<at 0.5108 1>,<at 0.5778 2>,<at 0.9182 2>,<at 1 1>]}
       RedCurve {keys [<at 0 10>,<at 0.2513 10>,<at 0.3322 1>,<at 0.5108 1>,<at 0.5778 2>,<at 0.9182 2>,<at 1 1>]}
      }})
    ScaleRGBA=true
   )
  ]
  rendererClasses [&NiagaraSpriteRendererProperties]
  renderers [{rendererClass=&NiagaraSpriteRendererProperties}]
 }
 Glow2 {
  emitterUpdateScript [
   &EmitterState()
   &SpawnRate(SpawnRate=20, Use Spawn Probability=false)
  ]
  particleSpawnScript [
   &InitializeParticle(
    Color=&MultiplyLinearColor(Linear Color=$User.ColorChange, Scale Factor=<1 1 1 0.05>)
    Module.Write Color=true
    Module.Write Lifetime=true
    Module.Write Mass=true
    Module.Write Material Random=true
    Module.Write Position=false
    Module.Write Scale=false
    Module.Write SpriteRotation=false
    Module.Write SpriteSize=true
    Position=&SimulationPosition()
    Sprite Size=&Vector2DFromFloat(Value=&RandomRangeFloat(Minimum=21, Maximum=25))
   )
   &CameraOffset(Camera Offset Amount=&RandomRangeFloat(Minimum=-5, Maximum=15))
   &FN_SetSpawnGroup_v2(Number of Spawn Groups=4)
   &BoxLocation(
    Box Size=<1 1 1>
    Coordinate Space:ENiagaraCoordinateSpace=::Simulation
    MaskSpawnGroup=false
    Module.Override Seed=false
    Rotate Box=false
    Surface Only=false
   )
  ]
  particleUpdateScript [
   &ParticleState()
   &SolveForcesAndVelocity(Clamp Velocity=false, Limit Acceleration=false)
   &ScaleColor(
    Scale Alpha=&FloatFromCurve(FloatCurve=DataInterface() {propertyValues {Curve {keys [<at 0 0>,<at 0.0767 1.0482>,<at 1 0>]}}})
    ScaleA=true
    ScaleRGB=true
   )
   &SkeletalMeshLocation(
    Bone / Socket Index=$Particles.SpawnGroup
    Inherit Velocity=false
    MaskSpawnGroup=false
    Mesh Sampling Type=::NewEnumerator5
    Sampled Position Offset:NiagaraPosition=$Particles.Initial.Position
    Skeletal Mesh=$User.SkltnMesh
    Socket Sampling Mode=::NewEnumerator5
   )
  ]
  rendererClasses [&NiagaraSpriteRendererProperties]
  renderers [{rendererClass=&NiagaraSpriteRendererProperties}]
 }
 LightLeft {
  emitterUpdateScript [
   &EmitterState()
   &SpawnRate(Use Spawn Probability=false) {enabled=false}
   &SpawnBurst_Instantaneous(Spawn Count=1, Spawn Time=&RandomRangeFloat(Minimum=0, Maximum=0.5), Use Spawn Probability=false)
  ]
  particleSpawnScript [&InitializeParticle(
    Lifetime=&RandomRangeFloat(Minimum=0.2, Maximum=0.46)
    Module.Write Color=true
    Module.Write Lifetime=true
    Module.Write Mass=true
    Module.Write Material Random=true
    Module.Write Position=true
    Module.Write Scale=false
    Module.Write SpriteRotation=false
    Module.Write SpriteSize=true
    Sprite Size=<3 3>
   )]
  particleUpdateScript [
   &ParticleState()
   &SkeletalMeshLocation(
    Bone / Socket Index=1
    Inherit Velocity=false
    MaskSpawnGroup=false
    Mesh Sampling Type=::NewEnumerator5
    Sampled Position Offset=<25 5 -6>
    Sampled Position Offset Coordinate Space:ENiagaraMeshTransforms=::NewEnumerator3
    Skeletal Mesh=$User.SkeletalMeshUp
    Socket Sampling Mode=::NewEnumerator5
   )
   &FN_Light_v3(
    Color=$User.ColorChange
    Create Color Attribute=true
    Create Exponent Attribute=true
    Create Radius Attribute=true
    Enable Particle Light Enable=true
    Exponent=3
    Light Enabled=true
    Radius=20
    Scale Color:NiagaraFloat=&FloatFromCurve(FloatCurve=DataInterface() {propertyValues {Curve {keys [<at 0 5>,<at 0.2105 5>,<at 0.2821 0.3547>,<at 0.4424 0.373>,<at 0.514 1.418>,<at 0.7086 1.418>,<at 0.7487 0.6113>]}}})
   )
  ]
  rendererClasses [&NiagaraSpriteRendererProperties, &NiagaraLightRendererProperties]
  renderers [
   {rendererClass=&NiagaraSpriteRendererProperties}
   {rendererClass=&NiagaraLightRendererProperties, rendererIndex=1}
  ]
 }
 LightRight {
  emitterUpdateScript [
   &EmitterState()
   &SpawnRate(Use Spawn Probability=false) {enabled=false}
   &SpawnBurst_Instantaneous(Spawn Count=1, Spawn Time=&RandomRangeFloat(Minimum=0, Maximum=0.5), Use Spawn Probability=false)
  ]
  particleSpawnScript [&InitializeParticle(
    Lifetime=&RandomRangeFloat(Minimum=0.2, Maximum=0.46)
    Module.Write Color=true
    Module.Write Lifetime=true
    Module.Write Mass=true
    Module.Write Material Random=true
    Module.Write Position=true
    Module.Write Scale=false
    Module.Write SpriteRotation=false
    Module.Write SpriteSize=true
    Sprite Size=<3 3>
   )]
  particleUpdateScript [
   &ParticleState()
   &SkeletalMeshLocation(
    Bone / Socket Index=1
    Inherit Velocity=false
    MaskSpawnGroup=false
    Mesh Sampling Type=::NewEnumerator5
    Sampled Position Offset=<-16 0 -6>
    Sampled Position Offset Coordinate Space:ENiagaraMeshTransforms=::NewEnumerator3
    Skeletal Mesh=$User.SkeletalMeshUp
    Socket Sampling Mode=::NewEnumerator5
   )
   &FN_Light_v3(
    Color=$User.ColorChange
    Create Color Attribute=true
    Create Exponent Attribute=true
    Create Radius Attribute=true
    Enable Particle Light Enable=true
    Exponent=3
    Light Enabled=true
    Radius=20
    Scale Color:NiagaraFloat=&FloatFromCurve(FloatCurve=DataInterface() {propertyValues {Curve {keys [<at 0 5>,<at 0.2105 5>,<at 0.2821 0.3547>,<at 0.4424 0.373>,<at 0.514 1.418>,<at 0.7086 1.418>,<at 0.7487 0.6113>]}}})
   )
  ]
  rendererClasses [&NiagaraSpriteRendererProperties, &NiagaraLightRendererProperties]
  renderers [
   {rendererClass=&NiagaraSpriteRendererProperties}
   {rendererClass=&NiagaraLightRendererProperties, rendererIndex=1}
  ]
 }
}
summary {
 systemName=NS_Backpack_Spacefighter_F_Idle_Pink
 emitters [
  {emitterName=Glow2, rendererClasses [&NiagaraSpriteRendererProperties]}
  {emitterName=Burst001, rendererClasses [&NiagaraSpriteRendererProperties]}
  {
   emitterName=BoneFlow_ArcUp
   rendererClasses [&NiagaraRibbonRendererProperties, &NiagaraSpriteRendererProperties]
  }
  {emitterName=BoneFlow_ArcBelow, rendererClasses [&NiagaraRibbonRendererProperties]}
  {emitterName=LightLeft, rendererClasses [&NiagaraSpriteRendererProperties, &NiagaraLightRendererProperties]}
  {emitterName=LightRight, rendererClasses [&NiagaraSpriteRendererProperties, &NiagaraLightRendererProperties]}
 ]
 userVariables [
  User.ColorChange {type=LinearColor, description="", defaultValue {struct=LinearColor, value=<rgba 1 0.0365 0.6518 1>}}
  User.SkeletalMeshBelow {
   type=NiagaraDataInterfaceSkeletalMesh
   defaultValue {
    struct=NiagaraExt_VariableValue_DataInterface
    value {dataInterface=&self:NiagaraDataInterfaceSkeletalMesh_1}
   }
  }
  User.SkeletalMeshUp {
   type=NiagaraDataInterfaceSkeletalMesh
   defaultValue {
    struct=NiagaraExt_VariableValue_DataInterface
    value {dataInterface=&self:NiagaraDataInterfaceSkeletalMesh_0}
   }
  }
  User.SkltnMesh {
   type=NiagaraDataInterfaceSkeletalMesh
   defaultValue {
    struct=NiagaraExt_VariableValue_DataInterface
    value {dataInterface=&self:NiagaraDataInterfaceSkeletalMesh_6}
   }
  }
 ]
}
